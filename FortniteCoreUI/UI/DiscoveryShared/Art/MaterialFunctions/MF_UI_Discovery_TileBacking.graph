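# UI material function for Discovery tile backings that draws an animated triangle pattern blended between two input colors.
# A fixed-angle rotated, tiled triangle texture scrolls with a falloff texture, alongside a separately panning blurred noise mask.
# Subtraction and smoothstep turn these into hard-edged triangles, faded by a vertical UV gradient.
# It outputs the color blend, masked pattern and unfaded pattern; a final gradient branch is left unconnected.
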MaterialFunction /FortniteCoreUI/UI/DiscoveryShared/Art/MaterialFunctions/MF_UI_Discovery_TileBacking
graph {nodes [
  n1 = Constant3Vector() {constant=<rgba 0 0 0.0118 1>}
  n2 = Constant3Vector() {constant=<rgba 0 0.0232 0.2542 1>}
  n3 = NamedRerouteUsage()
  FunctionOutput(None=LinearInterpolate(
    A=FunctionInput(Preview=@n1) {bUsePreviewValueAsDefault=true, description=Set the background color, inputName=Color A, sortPriority=1}
    B=FunctionInput(Preview=@n2) {bUsePreviewValueAsDefault=true, description=Set the triangles color, inputName=Color B, sortPriority=2}
    Alpha=@n3
   ))
  FunctionOutput(None=@n3) {outputName=Masked Pattern, sortPriority=1}
  n4 = Subtract(A=NamedRerouteUsage(), B=NamedRerouteUsage())
  FunctionOutput(None=Saturate(None=SmoothStep(Value=@n4) {constMax=0.2, constMin=0.08})) {outputName=Pattern, sortPriority=2}
  NamedRerouteDeclaration(None=FunctionInput(Preview=MaterialFunctionCall() {materialFunction=&GetUserInterfaceUV}.Normalized UV) {bUsePreviewValueAsDefault=true, inputName=UV, inputType=FunctionInput_Vector2}) {name=UV, nodeColor=<rgba 0 0 0 1>}
  n5 = Multiply(A=NamedRerouteUsage(), B=Constant() {desc=Noise Scale, r=2})
  n6 = Constant2Vector() {desc=Noise Speed, g=0.12, r=0.05}
  n7 = Multiply(
   A=TextureSample(
    UVs=Panner(Coordinate=@n5, Speed=@n6)
    Tex=TextureObject() {samplerType=SAMPLERTYPE_Grayscale, texture=&LowResBlurredNoise}
   ) {samplerType=SAMPLERTYPE_Grayscale, texture=&LowResBlurredNoise}.R
   B=Constant() {desc=Noise Multiplier, r=2}
  )
  n8 = Constant() {desc=Noise Power, r=3}
  NamedRerouteDeclaration(None=Saturate(None=Power(Base=@n7, Exp=@n8))) {name=Noise Mask, nodeColor=<rgba 0 0 0 1>}
  n9 = Rotator(Coordinate=NamedRerouteUsage(), Time=Constant() {desc=Triangles Rotation, r=0.0625}) {speed=6.2832}
  n10 = Constant() {desc=Triangles Tiling, r=12}
  n11 = Multiply(A=@n9, B=@n10) {constB=3}
  n12 = Panner(Coordinate=NamedRerouteUsage(), Speed=Constant2Vector() {g=0.0625, r=-0.0625})
  NamedRerouteDeclaration(None=TextureSample(UVs=Add(A=@n11, B=@n12), Tex=TextureObject() {texture=/Game/Global/Textures/Default/Masks_and_Gradients/T-BattlePass-Tile-Triangles}).RGB) {name=Triangle Pattern, nodeColor=<rgba 0 0 0 1>}
  n13 = Multiply(A=@n10) {constB=0.25}
  n14 = TextureSample(
   UVs=Add(A=Multiply(A=@n12) {constB=0.25}, B=Multiply(A=@n13, B=@n9))
   Tex=TextureObject() {samplerType=SAMPLERTYPE_Masks, texture=/FortniteCoreUI/UI/DiscoveryShared/Art/T_UI_Discovery_TriangleFalloff}
  )
  n15 = Add(A=Add(A=@n14.R, B=@n14.R), B=Multiply(A=@n14.B, B=Constant() {r=0.5}))
  NamedRerouteDeclaration(None=Multiply(A=Saturate(None=@n15), B=NamedRerouteUsage())) {name=Triangle Mask, nodeColor=<rgba 0 0 0 1>}
  n16 = NamedRerouteUsage()
  n17 = MaterialFunctionCall() {materialFunction=&GetUserInterfaceUV}
  n18 = Power(Base=Multiply(A=ComponentMask(None=@n17.Normalized UV) {g=true}) {constB=1.25}) {constExponent=0.25}
  n19 = LinearInterpolate(A=NamedRerouteUsage(), Alpha=OneMinus(None=Saturate(None=@n18))) {constAlpha=0}
  n20 = Saturate(None=SmoothStep(Value=Subtract(A=@n16, B=@n19)) {constMax=0.2, constMin=0.08})
  n21 = Power(Base=OneMinus(None=@n19)) {constExponent=6}
  NamedRerouteDeclaration(None=Saturate(None=Multiply(A=@n20, B=@n21))) {name=Alpha, nodeColor=<rgba 0 0 0 1>}
  n22 = Subtract(A=NamedRerouteUsage(), B=NamedRerouteUsage())
  n23 = Subtract(A=ComponentMask(None=MaterialFunctionCall() {materialFunction=&GetUserInterfaceUV}.Normalized UV) {g=true}) {constB=-0.2}
  n24 = Saturate(None=Power(Base=Multiply(A=@n23) {constB=1.25}))
  Saturate(None=SmoothStep(Value=Multiply(A=@n22, B=@n24)) {constMax=0.2, constMin=0.08})
 ]}
